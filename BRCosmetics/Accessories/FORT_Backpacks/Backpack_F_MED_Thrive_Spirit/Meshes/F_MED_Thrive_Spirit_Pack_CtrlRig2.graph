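# Control Rig for the Thrive Spirit backpack cosmetic that procedurally animates its snake tails, heads, right jaw and tongues.
# BeginExecution runs ten ChainHarmonics units in sequence, each driving a `dyn_` bone chain with its own wave, speed and curve settings.
# Only wave motion is enabled, with pendulum and reach off, and debug drawing is on.
# Three extra tongue ChainHarmonics nodes are left unconnected to execution.
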ControlRigBlueprint /BRCosmetics/Accessories/FORT_Backpacks/Backpack_F_MED_Thrive_Spirit/Meshes/F_MED_Thrive_Spirit_Pack_CtrlRig2
graphs [
 RigVMFunctionLibrary {links [], nodes []}
 RigVMModel {nodes [
   RigUnit_BeginExecution = &RigUnit_BeginExecution()
   /Script/ControlRig.RigUnit_ChainHarmonics(
    ChainRoot=dyn_snakeTail_00_rgt_MSTR
    DrawWorldOffset=(Rotation=(X=0,Y=0,Z=0,W=1),Translation=(X=0,Y=0,Z=0),Scale3D=(X=1,Y=1,Z=1))
    Pendulum=(bEnabled=False,PendulumStiffness=2,PendulumGravity=(X=0,Y=0,Z=0),PendulumBlend=0.75,PendulumDrag=0.98,PendulumMinimum=0,PendulumMaximum=0.1,PendulumEase=Linear,UnwindAxis=(X=0,Y=1,Z=0),UnwindMinimum=0.2,UnwindMaximum=0.05)
    Reach=(bEnabled=False,ReachTarget=(X=0,Y=0,Z=0),ReachAxis=(X=5,Y=0,Z=0),ReachMinimum=0,ReachMaximum=0,ReachEase=SineEaseInOut)
    Speed=(X=0.05,Y=0.05,Z=0.05)
    Wave=(bEnabled=True,WaveFrequency=(X=0.5,Y=3,Z=0.2),WaveAmplitude=(X=0.25,Y=0.5,Z=2),WaveOffset=(X=1,Y=0,Z=2),WaveNoise=(X=6,Y=6,Z=6),WaveMinimum=0,WaveMaximum=0.75,WaveEase=SineEaseIn)
    WaveCurve="(EditorCurveData=(Keys=((InterpMode=RCIM_Linear,TangentMode=RCTM_Auto,TangentWeightMode=RCTWM_WeightedNone,Time=-0.0198,Value=0.4929,ArriveTangent=0,ArriveTangentWeight=0,LeaveTangent=0,LeaveTangentWeight=0),(InterpMode=RCIM_Linear,TangentMode=RCTM_Auto,Ta ...(truncated)"
    bDrawDebug=True
    ExecutePin=@RigUnit_BeginExecution.ExecutePin
   ) {is=struct RigVMUnitNode}
   > &RigUnit_ChainHarmonics(
    ChainRoot=dyn_snakeTail_04_rgt
    DrawWorldOffset=(Rotation=(X=0,Y=0,Z=0,W=1),Translation=(X=0,Y=0,Z=0),Scale3D=(X=1,Y=1,Z=1))
    Pendulum=(bEnabled=False,PendulumStiffness=2,PendulumGravity=(X=0,Y=0,Z=0),PendulumBlend=0.75,PendulumDrag=0.98,PendulumMinimum=0,PendulumMaximum=0.1,PendulumEase=Linear,UnwindAxis=(X=0,Y=1,Z=0),UnwindMinimum=0.2,UnwindMaximum=0.05)
    Reach=(bEnabled=False,ReachTarget=(X=0,Y=0,Z=0),ReachAxis=(X=5,Y=0,Z=0),ReachMinimum=0,ReachMaximum=0,ReachEase=SineEaseInOut)
    Speed=(X=0.05,Y=0.05,Z=0.05)
    Wave=(bEnabled=True,WaveFrequency=(X=0.5,Y=3,Z=0.2),WaveAmplitude=(X=0.25,Y=0.5,Z=2),WaveOffset=(X=1,Y=0,Z=2),WaveNoise=(X=6,Y=6,Z=6),WaveMinimum=0,WaveMaximum=0.75,WaveEase=SineEaseIn)
    WaveCurve="(EditorCurveData=(Keys=((InterpMode=RCIM_Linear,TangentMode=RCTM_Auto,TangentWeightMode=RCTWM_WeightedNone,Time=-0.0198,Value=0.4929,ArriveTangent=0,ArriveTangentWeight=0,LeaveTangent=0,LeaveTangentWeight=0),(InterpMode=RCIM_Linear,TangentMode=RCTM_Auto,Ta ...(truncated)"
    bDrawDebug=True
   )
   > &RigUnit_ChainHarmonics(
    ChainRoot=dyn_snakeTail_08_rgt
    DrawWorldOffset=(Rotation=(X=0,Y=0,Z=0,W=1),Translation=(X=0,Y=0,Z=0),Scale3D=(X=1,Y=1,Z=1))
    Pendulum=(bEnabled=False,PendulumStiffness=2,PendulumGravity=(X=0,Y=0,Z=0),PendulumBlend=0.75,PendulumDrag=0.98,PendulumMinimum=0,PendulumMaximum=0.1,PendulumEase=Linear,UnwindAxis=(X=0,Y=1,Z=0),UnwindMinimum=0.2,UnwindMaximum=0.05)
    Reach=(bEnabled=False,ReachTarget=(X=0,Y=0,Z=0),ReachAxis=(X=5,Y=0,Z=0),ReachMinimum=0,ReachMaximum=0,ReachEase=SineEaseInOut)
    Speed=(X=0.05,Y=0.5,Z=0.05)
    Wave=(bEnabled=True,WaveFrequency=(X=-0.5,Y=0.5,Z=-0.2),WaveAmplitude=(X=-0.25,Y=-0.5,Z=-2),WaveOffset=(X=1,Y=0,Z=2),WaveNoise=(X=2,Y=2,Z=2),WaveMinimum=0,WaveMaximum=0.75,WaveEase=SineEaseIn)
    WaveCurve="(EditorCurveData=(Keys=((InterpMode=RCIM_Linear,TangentMode=RCTM_Auto,TangentWeightMode=RCTWM_WeightedNone,Time=-0.0198,Value=0.4929,ArriveTangent=0,ArriveTangentWeight=0,LeaveTangent=0,LeaveTangentWeight=0),(InterpMode=RCIM_Linear,TangentMode=RCTM_Auto,Ta ...(truncated)"
    bDrawDebug=True
   )
   > &RigUnit_ChainHarmonics(
    ChainRoot=dyn_snakeTail_00_lft_MSTR
    DrawWorldOffset=(Rotation=(X=0,Y=0,Z=0,W=1),Translation=(X=0,Y=0,Z=0),Scale3D=(X=1,Y=1,Z=1))
    Pendulum=(bEnabled=False,PendulumStiffness=2,PendulumGravity=(X=0,Y=0,Z=0),PendulumBlend=0.75,PendulumDrag=0.9,PendulumMinimum=0,PendulumMaximum=1,PendulumEase=Linear,UnwindAxis=(X=0,Y=1,Z=0),UnwindMinimum=0.2,UnwindMaximum=0.05)
    Reach=(bEnabled=False,ReachTarget=(X=0,Y=0,Z=0),ReachAxis=(X=0.5,Y=0.5,Z=0.5),ReachMinimum=0,ReachMaximum=0,ReachEase=SineEaseInOut)
    Speed=(X=0.1,Y=0.05,Z=0.1)
    Wave=(bEnabled=True,WaveFrequency=(X=5,Y=3,Z=5),WaveAmplitude=(X=0.5,Y=2,Z=0.5),WaveOffset=(X=10,Y=0,Z=2),WaveNoise=(X=8,Y=8,Z=15),WaveMinimum=0.3,WaveMaximum=0.75,WaveEase=SineEaseIn)
    WaveCurve="(EditorCurveData=(Keys=((InterpMode=RCIM_Linear,TangentMode=RCTM_Auto,TangentWeightMode=RCTWM_WeightedNone,Time=0,Value=0,ArriveTangent=0,ArriveTangentWeight=0,LeaveTangent=0,LeaveTangentWeight=0),(InterpMode=RCIM_Linear,TangentMode=RCTM_Auto,TangentWeight ...(truncated)"
    bDrawDebug=True
   )
   > &RigUnit_ChainHarmonics(
    ChainRoot=dyn_snakeHead_00_mid_MSTR
    DrawWorldOffset=(Rotation=(X=0,Y=0,Z=0,W=1),Translation=(X=0,Y=0,Z=0),Scale3D=(X=1,Y=1,Z=1))
    Pendulum=(bEnabled=False,PendulumStiffness=2,PendulumGravity=(X=0,Y=0,Z=0),PendulumBlend=0.75,PendulumDrag=0.98,PendulumMinimum=0,PendulumMaximum=0.1,PendulumEase=Linear,UnwindAxis=(X=0,Y=1,Z=0),UnwindMinimum=0.2,UnwindMaximum=0.05)
    Reach=(bEnabled=False,ReachTarget=(X=0,Y=0,Z=0),ReachAxis=(X=2.5,Y=0,Z=0),ReachMinimum=0,ReachMaximum=0,ReachEase=SineEaseInOut)
    Speed=(X=0.1,Y=0.3,Z=0.2)
    Wave=(bEnabled=True,WaveFrequency=(X=2.5,Y=0.2,Z=0.2),WaveAmplitude=(X=1,Y=1,Z=1),WaveOffset=(X=1,Y=0,Z=2),WaveNoise=(X=5,Y=5,Z=5),WaveMinimum=-0.3,WaveMaximum=0.8,WaveEase=SineEaseIn)
    WaveCurve="(EditorCurveData=(Keys=((InterpMode=RCIM_Linear,TangentMode=RCTM_Auto,TangentWeightMode=RCTWM_WeightedNone,Time=-0.0198,Value=0.4929,ArriveTangent=0,ArriveTangentWeight=0,LeaveTangent=0,LeaveTangentWeight=0),(InterpMode=RCIM_Linear,TangentMode=RCTM_Auto,Ta ...(truncated)"
    bDrawDebug=True
   )
   > &RigUnit_ChainHarmonics(
    ChainRoot=dyn_snakeTail_03_lft_MSTR
    DrawWorldOffset=(Rotation=(X=0,Y=0,Z=0,W=1),Translation=(X=0,Y=0,Z=0),Scale3D=(X=1,Y=1,Z=1))
    Pendulum=(bEnabled=False,PendulumStiffness=2,PendulumGravity=(X=0,Y=0,Z=0),PendulumBlend=0.75,PendulumDrag=0.98,PendulumMinimum=0,PendulumMaximum=0.1,PendulumEase=Linear,UnwindAxis=(X=0,Y=1,Z=0),UnwindMinimum=0.2,UnwindMaximum=0.05)
    Reach=(bEnabled=False,ReachTarget=(X=0,Y=0,Z=0),ReachAxis=(X=5,Y=0,Z=0),ReachMinimum=0,ReachMaximum=0,ReachEase=SineEaseInOut)
    Speed=(X=0.25,Y=0.25,Z=0.25)
    Wave=(bEnabled=True,WaveFrequency=(X=2.5,Y=5,Z=0.2),WaveAmplitude=(X=0.1,Y=0.1,Z=0.1),WaveOffset=(X=1,Y=0,Z=2),WaveNoise=(X=5,Y=5,Z=5),WaveMinimum=0,WaveMaximum=0.75,WaveEase=SineEaseIn)
    WaveCurve="(EditorCurveData=(Keys=((InterpMode=RCIM_Linear,TangentMode=RCTM_Auto,TangentWeightMode=RCTWM_WeightedNone,Time=-0.0198,Value=0.4929,ArriveTangent=0,ArriveTangentWeight=0,LeaveTangent=0,LeaveTangentWeight=0),(InterpMode=RCIM_Linear,TangentMode=RCTM_Auto,Ta ...(truncated)"
    bDrawDebug=True
   )
   > &RigUnit_ChainHarmonics(
    ChainRoot=dyn_snakeJaw_00_rgt
    DrawWorldOffset=(Rotation=(X=0,Y=0,Z=0,W=1),Translation=(X=0,Y=0,Z=0),Scale3D=(X=1,Y=1,Z=1))
    Pendulum=(bEnabled=False,PendulumStiffness=2,PendulumGravity=(X=0,Y=0,Z=0),PendulumBlend=0.75,PendulumDrag=0.98,PendulumMinimum=0,PendulumMaximum=0.1,PendulumEase=Linear,UnwindAxis=(X=0,Y=1,Z=0),UnwindMinimum=0.2,UnwindMaximum=0.05)
    Reach=(bEnabled=False,ReachTarget=(X=0,Y=0,Z=0),ReachAxis=(X=5,Y=0,Z=0),ReachMinimum=0,ReachMaximum=0,ReachEase=SineEaseInOut)
    Speed=(X=0,Y=1,Z=0)
    Wave=(bEnabled=True,WaveFrequency=(X=0,Y=3,Z=0),WaveAmplitude=(X=0,Y=1,Z=0),WaveOffset=(X=0,Y=1,Z=0),WaveNoise=(X=0,Y=0.2,Z=0),WaveMinimum=0.5,WaveMaximum=0.6,WaveEase=SineEaseIn)
    WaveCurve="(EditorCurveData=(Keys=((InterpMode=RCIM_Linear,TangentMode=RCTM_Auto,TangentWeightMode=RCTWM_WeightedNone,Time=-0.0198,Value=0.4502,ArriveTangent=0,ArriveTangentWeight=0,LeaveTangent=0,LeaveTangentWeight=0),(InterpMode=RCIM_Linear,TangentMode=RCTM_Auto,Ta ...(truncated)"
    bDrawDebug=True
   )
   > &RigUnit_ChainHarmonics(
    ChainRoot=dyn_snakeTongue_00_rgt
    DrawWorldOffset=(Rotation=(X=0,Y=0,Z=0,W=1),Translation=(X=0,Y=0,Z=0),Scale3D=(X=1,Y=1,Z=1))
    Pendulum=(bEnabled=False,PendulumStiffness=2,PendulumGravity=(X=0,Y=0,Z=0),PendulumBlend=0.75,PendulumDrag=0.98,PendulumMinimum=0,PendulumMaximum=0.1,PendulumEase=Linear,UnwindAxis=(X=0,Y=1,Z=0),UnwindMinimum=0.2,UnwindMaximum=0.05)
    Reach=(bEnabled=False,ReachTarget=(X=0,Y=0,Z=0),ReachAxis=(X=5,Y=0,Z=0),ReachMinimum=0,ReachMaximum=0,ReachEase=SineEaseInOut)
    Speed=(X=1.2,Y=2.2,Z=1.2)
    Wave=(bEnabled=True,WaveFrequency=(X=0,Y=0,Z=0),WaveAmplitude=(X=0,Y=0.5,Z=0),WaveOffset=(X=1,Y=0,Z=2),WaveNoise=(X=0,Y=0.5,Z=0.2),WaveMinimum=0.2,WaveMaximum=0.75,WaveEase=ElasticEaseInOut)
    WaveCurve="(EditorCurveData=(Keys=((InterpMode=RCIM_Linear,TangentMode=RCTM_Auto,TangentWeightMode=RCTWM_WeightedNone,Time=-0.0198,Value=0.49,ArriveTangent=0,ArriveTangentWeight=0,LeaveTangent=0,LeaveTangentWeight=0),(InterpMode=RCIM_Linear,TangentMode=RCTM_Auto,Tang ...(truncated)"
    bDrawDebug=True
   )
   > &RigUnit_ChainHarmonics(
    ChainRoot=dyn_snakeTongue_00_mid
    DrawWorldOffset=(Rotation=(X=0,Y=0,Z=0,W=1),Translation=(X=0,Y=0,Z=0),Scale3D=(X=1,Y=1,Z=1))
    Pendulum=(bEnabled=false,PendulumStiffness=2,PendulumGravity=(X=0,Y=0,Z=0),PendulumBlend=1,PendulumDrag=0.98,PendulumMinimum=0,PendulumMaximum=1,PendulumEase=Linear,UnwindAxis=(X=0,Y=1,Z=0),UnwindMinimum=0.2,UnwindMaximum=0.05)
    Reach=(bEnabled=False,ReachTarget=(X=0,Y=0,Z=0),ReachAxis=(X=5,Y=0,Z=0),ReachMinimum=0,ReachMaximum=0,ReachEase=SineEaseInOut)
    Speed=(X=0.2,Y=0.2,Z=0.5)
    Wave=(bEnabled=True,WaveFrequency=(X=0,Y=0.5,Z=1),WaveAmplitude=(X=0,Y=1,Z=1),WaveOffset=(X=1,Y=0,Z=2),WaveNoise=(X=0,Y=0.1,Z=0.4),WaveMinimum=0.2,WaveMaximum=0.6,WaveEase=SineEaseIn)
    WaveCurve="(EditorCurveData=(Keys=((InterpMode=RCIM_Linear,TangentMode=RCTM_Auto,TangentWeightMode=RCTWM_WeightedNone,Time=-0.0003,Value=1,ArriveTangent=0,ArriveTangentWeight=0,LeaveTangent=0,LeaveTangentWeight=0),(InterpMode=RCIM_Linear,TangentMode=RCTM_Auto,Tangent ...(truncated)"
    bDrawDebug=True
   )
   > &RigUnit_ChainHarmonics(
    ChainRoot=dyn_snakeTongue_00_lft
    DrawWorldOffset=(Rotation=(X=0,Y=0,Z=0,W=1),Translation=(X=0,Y=0,Z=0),Scale3D=(X=1,Y=1,Z=1))
    Pendulum=(bEnabled=False,PendulumStiffness=2,PendulumGravity=(X=0,Y=0,Z=0),PendulumBlend=0.75,PendulumDrag=0.98,PendulumMinimum=0,PendulumMaximum=0.1,PendulumEase=Linear,UnwindAxis=(X=0,Y=1,Z=0),UnwindMinimum=0.2,UnwindMaximum=0.05)
    Reach=(bEnabled=False,ReachTarget=(X=0,Y=0,Z=0),ReachAxis=(X=5,Y=0,Z=0),ReachMinimum=0,ReachMaximum=0,ReachEase=SineEaseInOut)
    Speed=(X=1,Y=2,Z=1)
    Wave=(bEnabled=True,WaveFrequency=(X=0,Y=0,Z=0),WaveAmplitude=(X=0,Y=0.5,Z=0),WaveOffset=(X=1,Y=0,Z=2),WaveNoise=(X=0,Y=0.5,Z=0.2),WaveMinimum=0.2,WaveMaximum=0.75,WaveEase=ElasticEaseInOut)
    WaveCurve="(EditorCurveData=(Keys=((InterpMode=RCIM_Linear,TangentMode=RCTM_Auto,TangentWeightMode=RCTWM_WeightedNone,Time=-0.0198,Value=0.49,ArriveTangent=0,ArriveTangentWeight=0,LeaveTangent=0,LeaveTangentWeight=0),(InterpMode=RCIM_Linear,TangentMode=RCTM_Auto,Tang ...(truncated)"
    bDrawDebug=True
   )
   &RigUnit_ChainHarmonics(
    ChainRoot=dyn_snakeTongue_00_mid
    DrawWorldOffset=(Rotation=(X=0,Y=0,Z=0,W=1),Translation=(X=0,Y=0,Z=0),Scale3D=(X=1,Y=1,Z=1))
    ExecutePin=()
    Pendulum=(bEnabled=false,PendulumStiffness=2,PendulumGravity=(X=0,Y=0,Z=0),PendulumBlend=1,PendulumDrag=0.98,PendulumMinimum=0,PendulumMaximum=1,PendulumEase=Linear,UnwindAxis=(X=0,Y=1,Z=0),UnwindMinimum=0.2,UnwindMaximum=0.05)
    Reach=(bEnabled=False,ReachTarget=(X=0,Y=0,Z=0),ReachAxis=(X=5,Y=0,Z=0),ReachMinimum=0,ReachMaximum=0,ReachEase=SineEaseInOut)
    Speed=(X=0.2,Y=0.2,Z=0.5)
    Wave=(bEnabled=True,WaveFrequency=(X=0,Y=0.5,Z=1),WaveAmplitude=(X=0,Y=1,Z=1),WaveOffset=(X=1,Y=0,Z=2),WaveNoise=(X=0,Y=0.1,Z=0.4),WaveMinimum=0.2,WaveMaximum=0.6,WaveEase=SineEaseIn)
    WaveCurve="(EditorCurveData=(Keys=((InterpMode=RCIM_Linear,TangentMode=RCTM_Auto,TangentWeightMode=RCTWM_WeightedNone,Time=-0.0003,Value=1,ArriveTangent=0,ArriveTangentWeight=0,LeaveTangent=0,LeaveTangentWeight=0),(InterpMode=RCIM_Linear,TangentMode=RCTM_Auto,Tangent ...(truncated)"
    bDrawDebug=True
   )
   &RigUnit_ChainHarmonics(
    ChainRoot=dyn_snakeTongue_00_lft
    DrawWorldOffset=(Rotation=(X=0,Y=0,Z=0,W=1),Translation=(X=0,Y=0,Z=0),Scale3D=(X=1,Y=1,Z=1))
    ExecutePin=()
    Pendulum=(bEnabled=false,PendulumStiffness=2,PendulumGravity=(X=0,Y=0,Z=0),PendulumBlend=1,PendulumDrag=0.98,PendulumMinimum=0,PendulumMaximum=1,PendulumEase=Linear,UnwindAxis=(X=0,Y=1,Z=0),UnwindMinimum=0.2,UnwindMaximum=0.05)
    Reach=(bEnabled=False,ReachTarget=(X=0,Y=0,Z=0),ReachAxis=(X=0,Y=0,Z=0),ReachMinimum=0,ReachMaximum=0,ReachEase=SineEaseInOut)
    Speed=(X=1,Y=2,Z=1)
    Wave=(bEnabled=True,WaveFrequency=(X=0.1,Y=1,Z=1),WaveAmplitude=(X=0.1,Y=0.5,Z=0.5),WaveOffset=(X=1,Y=0,Z=-2),WaveNoise=(X=0,Y=0.1,Z=0.4),WaveMinimum=0.2,WaveMaximum=0.6,WaveEase=SineEaseIn)
    WaveCurve="(EditorCurveData=(Keys=((InterpMode=RCIM_Linear,TangentMode=RCTM_Auto,TangentWeightMode=RCTWM_WeightedNone,Time=-0.0003,Value=1,ArriveTangent=0,ArriveTangentWeight=0,LeaveTangent=0,LeaveTangentWeight=0),(InterpMode=RCIM_Linear,TangentMode=RCTM_Auto,Tangent ...(truncated)"
    bDrawDebug=True
   )
   &RigUnit_ChainHarmonics(
    ChainRoot=dyn_snakeTongue_00_mid
    DrawWorldOffset=(Rotation=(X=0,Y=0,Z=0,W=1),Translation=(X=0,Y=0,Z=0),Scale3D=(X=1,Y=1,Z=1))
    ExecutePin=()
    Pendulum=(bEnabled=False,PendulumStiffness=2,PendulumGravity=(X=0,Y=0,Z=0),PendulumBlend=0.75,PendulumDrag=0.98,PendulumMinimum=0,PendulumMaximum=0.1,PendulumEase=Linear,UnwindAxis=(X=0,Y=1,Z=0),UnwindMinimum=0.2,UnwindMaximum=0.05)
    Reach=(bEnabled=False,ReachTarget=(X=0,Y=0,Z=0),ReachAxis=(X=5,Y=0,Z=0),ReachMinimum=0,ReachMaximum=0,ReachEase=SineEaseInOut)
    Speed=(X=1,Y=2,Z=1)
    Wave=(bEnabled=True,WaveFrequency=(X=0,Y=0,Z=0),WaveAmplitude=(X=0,Y=0.5,Z=0),WaveOffset=(X=1,Y=0,Z=2),WaveNoise=(X=0,Y=0.5,Z=0.2),WaveMinimum=0.2,WaveMaximum=0.75,WaveEase=ElasticEaseInOut)
    WaveCurve="(EditorCurveData=(Keys=((InterpMode=RCIM_Linear,TangentMode=RCTM_Auto,TangentWeightMode=RCTWM_WeightedNone,Time=-0.0198,Value=0.49,ArriveTangent=0,ArriveTangentWeight=0,LeaveTangent=0,LeaveTangentWeight=0),(InterpMode=RCIM_Linear,TangentMode=RCTM_Auto,Tang ...(truncated)"
    bDrawDebug=True
   )
  ]}
]
hierarchy [
 root {loc=<0 0 0>, rot=<0 0 0 1>}
 pack {parent=root, loc=<0 -0.7316 0>, rot=<0 0 0 1>}
 dyn_snakeHead_00_mid_MSTR {parent=pack, loc=<0.8706 -6.6077 -12.5246>, rot=<0.6143 -0.4334 -0.3801 -0.5388>}
 dyn_snakeHead_01_mid {parent=dyn_snakeHead_00_mid_MSTR, loc=<0 0 0>, rot=<0 0 0 1>}
 dyn_snakeTongue_00_mid {parent=dyn_snakeHead_01_mid, loc=<2.8243 0 0>, rot=<0.0115 -0.0839 -0.2038 -0.9753>}
 dyn_snakeTongue_01_mid {parent=dyn_snakeTongue_00_mid, loc=<1.6217 0 0>, rot=<-0.0045 -0.0393 -0.0435 -0.9983>}
 dyn_snakeTongue_02_mid {parent=dyn_snakeTongue_01_mid, loc=<1.3021 0 0>, rot=<-0.0124 -0.1774 0.0861 -0.9803>}
 dyn_snakeTail_00_rgt_MSTR {parent=pack, loc=<-10.9577 -8.769 15.6179>, rot=<0.2617 -0.8461 -0.4435 -0.1372>}
 dyn_snakeTail_01_rgt {parent=dyn_snakeTail_00_rgt_MSTR, loc=<0 0 0>, rot=<0 0 0 1>}
 dyn_snakeTail_02_rgt {parent=dyn_snakeTail_01_rgt, loc=<1.803 0 0>, rot=<0.3246 0.3573 0.2024 -0.852>}
 dyn_snakeTail_03_rgt {parent=dyn_snakeTail_02_rgt, loc=<3.7313 0 0>, rot=<0.0463 0.1607 0.219 -0.9613>}
 dyn_snakeTail_03_lft_MSTR {parent=pack, loc=<-3.5221 -5.3026 11.518>, rot=<0 0 0 1>}
 dyn_snakeTail_04_lft {parent=dyn_snakeTail_03_lft_MSTR, loc=<0 0 0>, rot=<-0.5277 0.4982 -0.4723 -0.5003>}
 dyn_snakeTail_05_lft {parent=dyn_snakeTail_04_lft, loc=<1.9464 0 0>, rot=<0.023 0.2634 -0.1389 -0.9544>}
 dyn_snakeTail_06_lft {parent=dyn_snakeTail_05_lft, loc=<3.3478 0 0>, rot=<0.1809 0.2872 -0.3054 -0.8897>}
 dyn_snakeTail_07_lft {parent=dyn_snakeTail_06_lft, loc=<2.9198 0 0>, rot=<0.8576 0.4406 -0.0536 -0.2598>}
 dyn_snakeTail_00_lft_MSTR {parent=pack, loc=<7.6118 -5.1481 1.6178>, rot=<0.8229 -0.2962 -0.1642 -0.4561>}
 dyn_snakeTail_01_lft {parent=dyn_snakeTail_00_lft_MSTR, loc=<0 0 0>, rot=<0 0 0 1>}
 dyn_snakeTail_02_lft {parent=dyn_snakeTail_01_lft, loc=<2.6059 0 0>, rot=<-0.1664 -0.0628 -0.3878 -0.9044>}
 dyn_snakeHead_00_lft {parent=dyn_snakeTail_02_lft, loc=<1.2029 -0.0518 -0.0006>, rot=<-0.0201 0.2127 -0.0777 -0.9738>}
 dyn_snakeJaw_00_lft {parent=dyn_snakeHead_00_lft, loc=<1.7251 0.006 -0.1862>, rot=<-0.0155 -0.1666 -0.0006 -0.9859>}
 dyn_snakeTongue_00_lft {parent=dyn_snakeJaw_00_lft, loc=<2.237 0.044 0.1972>, rot=<0.0438 -0.2921 0.0314 -0.9549>}
 dyn_snakeTongue_01_lft {parent=dyn_snakeTongue_00_lft, loc=<0.9099 -0.0267 -0.0044>, rot=<-0.0639 -0.5268 -0.0449 -0.8464>}
 dyn_snakeTongue_02_lft {parent=dyn_snakeTongue_01_lft, loc=<1.1012 0.0138 0>, rot=<0.0081 0.293 0.0455 -0.955>}
 dyn_snakeTail_04_rgt {parent=root, loc=<-14.5191 -5.5689 8.224>, rot=<0.3306 -0.2327 -0.5264 -0.748>}
 dyn_snakeTail_05_rgt {parent=dyn_snakeTail_04_rgt, loc=<4.2333 0 0>, rot=<0.0929 -0.2499 0.0111 -0.9637>}
 dyn_snakeTail_06_rgt {parent=dyn_snakeTail_05_rgt, loc=<3.969 0 0>, rot=<0.0098 -0.0463 -0.1574 -0.9864>}
 dyn_snakeTail_07_rgt {parent=dyn_snakeTail_06_rgt, loc=<3.4486 0 0>, rot=<-0.0382 -0.092 -0.4278 -0.8984>}
 dyn_snakeTail_08_rgtMSTR {parent=root, loc=<-13.6882 -1.9891 -3.7517>, rot=<0.1849 -0.9827 -0.0089 -0.0017>}
 dyn_snakeTail_08_rgt {parent=dyn_snakeTail_08_rgtMSTR, loc=<0 0 0>, rot=<0 0 0 1>}
 dyn_snakeTail_09_rgt {parent=dyn_snakeTail_08_rgt, loc=<2.4268 0 0>, rot=<-0.3148 -0.1522 0.0733 -0.934>}
 dyn_snakeTail_10_rgt {parent=dyn_snakeTail_09_rgt, loc=<1.9552 0 0>, rot=<-0.0896 -0.089 0.2491 -0.9602>}
 dyn_snakeTail_11_rgt {parent=dyn_snakeTail_10_rgt, loc=<2.3986 0 0>, rot=<0.0755 0.1961 0.1624 -0.9641>}
 dyn_snakeTail_12_rgt {parent=dyn_snakeTail_11_rgt, loc=<3.1177 0 0>, rot=<-0.0947 0.6194 0.0781 -0.7754>}
 dyn_snakeHead_00_rgt {parent=dyn_snakeTail_12_rgt, loc=<1.3687 0 0>, rot=<-0.0322 0.1134 0.0159 -0.9929>}
 dyn_snakeJaw_00_rgt {parent=dyn_snakeHead_00_rgt, loc=<0.8011 -0.0064 -0.3768>, rot=<0.0126 -0.1829 0.0106 -0.983>}
 dyn_snakeTongue_00_rgt {parent=dyn_snakeJaw_00_rgt, loc=<1.9982 0.0497 0.3715>, rot=<-0.0258 -0.2367 -0.0708 -0.9687>}
 dyn_snakeTongue_01_rgt {parent=dyn_snakeTongue_00_rgt, loc=<0.8266 0 -0.0872>, rot=<0.0142 0.2839 0.0304 -0.9583>}
 dyn_snakeTongue_02_rgt {parent=dyn_snakeTongue_01_rgt, loc=<0.4918 0 0>, rot=<-0.0194 0.3527 -0.0011 -0.9355>}
 dyn_snakeTongue_03_rgt {parent=dyn_snakeTongue_02_rgt, loc=<0.9033 0 0>, rot=<-0.0579 -0.3734 0.0233 -0.9256>}
 dyn_snakeTongue_04_rgt {parent=dyn_snakeTongue_03_rgt, loc=<0.448 0.0031 0.029>, rot=<-0.0497 -0.9016 0.0133 -0.4294>}
 dyn_snakeTongue_05_rgt {parent=dyn_snakeTongue_04_rgt, loc=<0 0 0.5392>, rot=<0 0 0 1>}
]
