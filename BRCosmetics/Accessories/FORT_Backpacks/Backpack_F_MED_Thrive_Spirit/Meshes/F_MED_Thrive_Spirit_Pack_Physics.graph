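# Physics asset for the Thrive Spirit backpack cosmetic, defining ragdoll-style bodies for its two dangling snakes around a large pack capsule.
# Capsules on left and right snake head, jaw and tail bones form several constraint chains with tight twist or swing limits.
# Some chains anchor to the pack, while right-side chains starting at tail 04 and tail 08 MSTR have no constraint to it.
# Only the left tail 07 joint rotates freely.
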PhysicsAsset /BRCosmetics/Accessories/FORT_Backpacks/Backpack_F_MED_Thrive_Spirit/Meshes/F_MED_Thrive_Spirit_Pack_Physics
bodies {
 dyn_snakeHead_00_lft [{
   shapeType=Capsule
   length=1.8221
   radius=1.3313
   shapeName=dyn_snakeHead_00_lft_capsule
   center=<1.7254 0.0335 -0.8843>
   rotation=<rot 0 -97.3105 -74.9083>
  }]
 dyn_snakeHead_00_rgt [{
   shapeType=Capsule
   length=1.6347
   radius=1.6842
   shapeName=dyn_snakeHead_00_rgt_capsule
   center=<1.2394 -0.0001 -1.04>
   rotation=<rot 50.0004 0 0>
  }]
 dyn_snakeJaw_00_lft [{
   shapeType=Capsule
   length=1.2946
   radius=1.1928
   shapeName=dyn_snakeJaw_00_lft_capsule
   center=<0.507 0.1628 0.1758>
   rotation=<rot 0 124.6187 -81.371>
  }]
 dyn_snakeTail_00_lft_MSTR [{shapeType=Capsule, length=0.505, radius=0.505, shapeName=dyn_snakeTail_00_lft_MSTR_capsule, rotation=<rot 0 -90 0>}]
 dyn_snakeTail_01_lft [{
   shapeType=Capsule
   length=0.9929
   radius=0.7276
   shapeName=dyn_snakeTail_01_lft_capsule
   center=<1.3753 -0.3255 0.0805>
   rotation=<rot 0 75.578 -82.2>
  }]
 dyn_snakeTail_01_rgt [{
   shapeType=Capsule
   length=0.8881
   radius=0.8336
   shapeName=dyn_snakeTail_01_rgt_capsule
   center=<0.4268 0.0229 -0.1928>
   rotation=<rot 17.9588 -171.3471 -90>
  }]
 dyn_snakeTail_02_lft [{
   shapeType=Capsule
   length=0.858
   radius=0.6947
   shapeName=dyn_snakeTail_02_lft_capsule
   center=<0.3912 0.032 0.1856>
   rotation=<rot 0 12.7271 -78.3608>
  }]
 dyn_snakeTail_02_rgt [{
   shapeType=Capsule
   length=2.3739
   radius=0.8978
   shapeName=dyn_snakeTail_02_rgt_capsule
   center=<2.2664 -0.1566 -0.1217>
   rotation=<rot 0 85.3681 -86.9688>
  }]
 dyn_snakeTail_03_rgt [{
   shapeType=Capsule
   length=1.6418
   radius=0.841
   shapeName=dyn_snakeTail_03_rgt_capsule
   center=<3.3225 0.0555 0.1049>
   rotation=<rot 0 83.9901 -83.2227>
  }]
 dyn_snakeTail_04_lft [{
   shapeType=Capsule
   length=0.9028
   radius=0.7935
   shapeName=dyn_snakeTail_04_lft_capsule
   center=<0.5847 -0.0908 0.0258>
   rotation=<rot 0 -9.7342 -69.0155>
  }]
 dyn_snakeTail_04_rgt [{shapeType=Capsule, length=0.505, radius=0.505, shapeName=dyn_snakeTail_04_rgt_capsule, rotation=<rot 0 -90 0>}]
 dyn_snakeTail_05_rgt [{
   shapeType=Capsule
   length=2.0269
   radius=0.8031
   shapeName=dyn_snakeTail_05_rgt_capsule
   center=<-0.973 0.2265 -0.5449>
   rotation=<rot 0 86.8791 -73.5541>
  }]
 dyn_snakeTail_06_rgt [{
   shapeType=Capsule
   length=1.052
   radius=0.8202
   shapeName=dyn_snakeTail_06_rgt_capsule
   center=<-0.8089 0.3206 0.029>
   rotation=<rot 0 60.1765 -88.432>
  }]
 dyn_snakeTail_07_lft [{
   shapeType=Capsule
   length=2.3023
   radius=1.1957
   shapeName=dyn_snakeTail_07_lft_capsule
   center=<0.8431 0.2455 0.0348>
   rotation=<rot 0 97.1077 -89.2011>
  }]
 dyn_snakeTail_07_rgt [{
   shapeType=Capsule
   length=1.5211
   radius=0.8796
   shapeName=dyn_snakeTail_07_rgt_capsule
   center=<-0.3646 0.1824 -0.0212>
   rotation=<rot 0 54.7041 -80.9418>
  }]
 dyn_snakeTail_08_rgt [{shapeType=Capsule, length=0.505, radius=0.505, shapeName=dyn_snakeTail_09_rgt_capsule, rotation=<rot -0.0001 -89.9999 -100.0002>}]
 dyn_snakeTail_08_rgtMSTR [{shapeType=Capsule, length=0.505, radius=0.505, shapeName=dyn_snakeTail_08_rgtMSTR_capsule, rotation=<rot 0 -90 0>}]
 dyn_snakeTail_09_rgt [{shapeType=Capsule, length=0.505, radius=0.505, shapeName=dyn_snakeTail_09_rgt_capsule, rotation=<rot -0.0001 -89.9999 -100.0002>}]
 dyn_snakeTail_10_rgt [{shapeType=Capsule, length=1.4441, radius=0.7424, shapeName=dyn_snakeTail_10_rgt_capsule, rotation=<rot 90 0 0>}]
 dyn_snakeTail_11_rgt [{
   shapeType=Capsule
   length=1.8977
   radius=0.8537
   shapeName=dyn_snakeTail_11_rgt_capsule
   center=<0.5501 0 -0.2002>
   rotation=<rot 69.9999 -0.0001 0>
  }]
 dyn_snakeTail_12_rgt [{
   shapeType=Capsule
   length=0.505
   radius=0.505
   shapeName=dyn_snakeTail_12_rgt_capsule
   center=<-0.2778 0 0.1011>
   rotation=<rot 59.9996 0 -0.0001>
  }]
 pack [{shapeType=Capsule, length=38.3288, radius=7.2001, center=<1.5077 -5.4242 -11.5628>, rotation=<rot -1.885 -90 -21.1394>}]
}
constraints [
 {bone1Name=dyn_snakeTail_08_rgt, bone2Name=dyn_snakeTail_08_rgtMSTR, swing1LimitDegrees=10, swing2LimitDegrees=10, twistLimitDegrees=5}
 {bone1Name=dyn_snakeTail_09_rgt, bone2Name=dyn_snakeTail_08_rgt, swing1LimitDegrees=10, swing2LimitDegrees=10, twistLimitDegrees=5}
 {bone1Name=dyn_snakeTail_10_rgt, bone2Name=dyn_snakeTail_09_rgt, swing1LimitDegrees=10, swing2LimitDegrees=10, twistLimitDegrees=5}
 {bone1Name=dyn_snakeTail_11_rgt, bone2Name=dyn_snakeTail_10_rgt, swing1LimitDegrees=10, swing2LimitDegrees=10, twistLimitDegrees=5}
 {bone1Name=dyn_snakeTail_12_rgt, bone2Name=dyn_snakeTail_11_rgt, swing1LimitDegrees=10, swing2LimitDegrees=10, twistLimitDegrees=5}
 {bone1Name=dyn_snakeHead_00_rgt, bone2Name=dyn_snakeTail_12_rgt, swing1LimitDegrees=10, swing2LimitDegrees=10, twistLimitDegrees=5}
 {bone1Name=dyn_snakeTail_00_lft_MSTR, bone2Name=pack, swing1LimitDegrees=45, swing2LimitDegrees=15, twistLimitDegrees=2}
 {bone1Name=dyn_snakeTail_01_lft, bone2Name=dyn_snakeTail_00_lft_MSTR, swing1LimitDegrees=45, swing2LimitDegrees=45, twistLimitDegrees=2}
 {bone1Name=dyn_snakeTail_02_lft, bone2Name=dyn_snakeTail_01_lft, swing1LimitDegrees=45, swing2LimitDegrees=45, twistLimitDegrees=2}
 {bone1Name=dyn_snakeHead_00_lft, bone2Name=dyn_snakeTail_02_lft, swing1LimitDegrees=45, swing2LimitDegrees=45, twistLimitDegrees=2}
 {bone1Name=dyn_snakeJaw_00_lft, bone2Name=dyn_snakeHead_00_lft, swing1LimitDegrees=45, swing2LimitDegrees=45, twistLimitDegrees=2}
 {bone1Name=dyn_snakeTail_02_rgt, bone2Name=dyn_snakeTail_01_rgt, swing1LimitDegrees=25, swing2LimitDegrees=15, twistLimitDegrees=2}
 {bone1Name=dyn_snakeTail_01_rgt, bone2Name=pack, swing1LimitDegrees=25, swing2LimitDegrees=15, twistLimitDegrees=2}
 {bone1Name=dyn_snakeTail_03_rgt, bone2Name=dyn_snakeTail_02_rgt, swing1LimitDegrees=25, swing2LimitDegrees=15, twistLimitDegrees=2}
 {bone1Name=dyn_snakeTail_05_rgt, bone2Name=dyn_snakeTail_04_rgt, swing1LimitDegrees=45, swing2LimitDegrees=15, twistLimitDegrees=2}
 {bone1Name=dyn_snakeTail_06_rgt, bone2Name=dyn_snakeTail_05_rgt, swing1LimitDegrees=45, swing2LimitDegrees=15, twistLimitDegrees=2}
 {bone1Name=dyn_snakeTail_07_rgt, bone2Name=dyn_snakeTail_06_rgt, swing1LimitDegrees=45, swing2LimitDegrees=15, twistLimitDegrees=2}
 {bone1Name=dyn_snakeTail_04_lft, bone2Name=pack, swing1LimitDegrees=10, swing2LimitDegrees=10, twistLimitDegrees=45}
 {
  bone1Name=dyn_snakeTail_07_lft
  bone2Name=dyn_snakeTail_04_lft
  swing1LimitDegrees=45
  swing1Motion=Free
  swing2LimitDegrees=45
  swing2Motion=Free
  twistLimitDegrees=45
  twistMotion=Free
 }
]
