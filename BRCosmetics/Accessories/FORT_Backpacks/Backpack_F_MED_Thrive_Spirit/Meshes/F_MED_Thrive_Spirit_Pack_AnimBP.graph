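# Animation Blueprint for the Thrive Spirit backpack, animating its snakes and toggling a material hide parameter from equipped ThriveFemale-tagged pickaxe and skydive events.
# Control Rig feeds additive bone offsets, noise-driven head and tongue sway, tongue trail controllers, then low-alpha rigid-body physics.
# Outside the front end, spine yaw computes backpack offsets for a final root-bone node whose alpha is zero.
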AnimBlueprint /BRCosmetics/Accessories/FORT_Backpacks/Backpack_F_MED_Thrive_Spirit/Meshes/F_MED_Thrive_Spirit_Pack_AnimBP
target_skeleton=&/F_MED_Thrive_Spirit_Pack_Skeleton
graphs [
 AnimGraph {nodes [
   Local Space Ref Pose()
   n1 = Control Rig(Source=Input Pose()) {
    alphaBoolBlend {blendOption=Linear}
    controlRigAssetReference {blueprintRigClass=&/F_MED_Thrive_Spirit_Pack_CtrlRig2.F_MED_Thrive_Spirit_Pack_CtrlRig2_C}
   }
   n2 = "Transform (Modify) Bone - Bone: dyn_snakeTongue_00_rgt"(
    ComponentPose=Local To Component(LocalPose=@n1)
    Alpha=0.85
    Translation="0,0,0"
    Rotation=Get Snake Movement() -> (Snake Movement:Rotator)
    Scale="1,1,1"
   ) {rotationMode=BMM_Additive, rotationSpace=BCS_BoneSpace, alphaBoolBlend {blendOption=Linear}, boneToModify {boneName=dyn_snakeTongue_00_rgt}}
   n3 = Trail controller - Bone: dyn_snakeTongue_05_rgt(ComponentPose=@n2, Alpha=0.6) {
    chainBoneAxis=Z
    chainLength=5
    stretchLimit=0.3
    rotationLimits [{limitMax=<180 180 180>}, {limitMax=<180 180 180>}, {limitMax=<180 180 180>}, {limitMax=<180 180 180>}, {limitMax=<180 180 180>}]
    rotationOffsets [<0 0 0>, <0 0 0>, <0 0 0>, <0 0 0>, <0 0 0>]
    trailBone {boneName=dyn_snakeTongue_05_rgt}
    trailRelaxationSpeed {editorCurveData {keys [<at 0 10 RCIM_Linear>,<at 0.2425 9.5109 RCIM_Linear>,<at 1 5 RCIM_Linear>]}}
   }
   n4 = "Transform (Modify) Bone - Bone: dyn_snakeTongue_01_lft"(ComponentPose=@n3, Alpha=0.5, Translation="0,0,0", Rotation="80,0,0", Scale="1,1,1") {
    rotationMode=BMM_Additive
    rotationSpace=BCS_BoneSpace
    alphaBoolBlend {blendOption=Linear}
    boneToModify {boneName=dyn_snakeTongue_01_lft}
    rotation {pitch=80}
   }
   n5 = Trail controller - Bone: dyn_snakeTongue_02_mid(ComponentPose=@n4, Alpha=0.5) {
    chainLength=3
    rotationLimits [{limitMax=<180 180 180>}, {limitMax=<180 180 180>}, {limitMax=<180 180 180>}]
    rotationOffsets [<0 0 0>, <0 0 0>, <0 0 0>]
    trailBone {boneName=dyn_snakeTongue_02_mid}
    trailRelaxationSpeed {editorCurveData {keys [<at 0 10 RCIM_Linear>,<at 0.2396 9.9839 RCIM_Linear>,<at 0.3091 9.9951 RCIM_Linear>]}}
   }
   n6 = "Transform (Modify) Bone - Bone: dyn_snakeJaw_00_rgt"(ComponentPose=@n5, Alpha=1, Translation="-0.2,0,-0.25", Rotation="-20,0,0", Scale="1,1,1") {
    rotationMode=BMM_Additive
    rotationSpace=BCS_BoneSpace
    translationMode=BMM_Additive
    translationSpace=BCS_BoneSpace
    alphaBoolBlend {blendOption=Linear}
    boneToModify {boneName=dyn_snakeJaw_00_rgt}
    rotation {pitch=-20}
    translation {x=-0.2, z=-0.25}
   }
   n7 = "Transform (Modify) Bone - Bone: dyn_snakeTongue_00_rgt"(ComponentPose=@n6, Alpha=1, Translation="0,0,0", Rotation="40,0,0", Scale="1,1,1") {
    rotationMode=BMM_Additive
    rotationSpace=BCS_BoneSpace
    alphaBoolBlend {blendOption=Linear}
    boneToModify {boneName=dyn_snakeTongue_00_rgt}
    rotation {pitch=40}
   }
   n8 = "Transform (Modify) Bone - Bone: dyn_snakeTongue_00_mid"(ComponentPose=@n7, Alpha=1, Translation="0,0,0", Rotation="-20,0,0", Scale="1,1,1") {
    rotationMode=BMM_Additive
    rotationSpace=BCS_BoneSpace
    alphaBoolBlend {blendOption=Linear}
    boneToModify {boneName=dyn_snakeTongue_00_mid}
    rotation {pitch=-20}
   }
   n9 = "Transform (Modify) Bone - Bone: dyn_snakeTongue_00_lft"(ComponentPose=@n8, Alpha=1, Translation="0,0,0", Rotation="60,0,0", Scale="1,1,1") {
    rotationMode=BMM_Additive
    rotationSpace=BCS_BoneSpace
    alphaBoolBlend {blendOption=Linear}
    boneToModify {boneName=dyn_snakeTongue_00_lft}
    rotation {pitch=60}
   }
   n10 = "Transform (Modify) Bone - Bone: dyn_snakeHead_00_rgt"(ComponentPose=@n9, Alpha=1, Translation="0,0,0", Rotation=Get Snake Movement(), Scale="1,1,1") {
    rotationMode=BMM_Additive
    rotationSpace=BCS_BoneSpace
    scaleMode=BMM_Additive
    scaleSpace=BCS_BoneSpace
    translationMode=BMM_Additive
    translationSpace=BCS_BoneSpace
    alphaBoolBlend {blendOption=Linear}
    boneToModify {boneName=dyn_snakeHead_00_rgt}
   }
   n11 = "Transform (Modify) Bone - Bone: dyn_snakeHead_01_mid"(ComponentPose=@n10, Alpha=1, Translation="0,0,0", Rotation=Get Snake Movement(), Scale="1,1,1") {
    rotationMode=BMM_Additive
    rotationSpace=BCS_BoneSpace
    scaleMode=BMM_Additive
    scaleSpace=BCS_BoneSpace
    translationMode=BMM_Additive
    translationSpace=BCS_BoneSpace
    alphaBoolBlend {blendOption=Linear}
    boneToModify {boneName=dyn_snakeHead_01_mid}
   }
   n12 = "Transform (Modify) Bone - Bone: dyn_snakeHead_01_mid"(ComponentPose=@n11, Alpha=1, Translation="0,0,0", Rotation="-15,0,0", Scale="1,1,1") {
    rotationMode=BMM_Additive
    rotationSpace=BCS_BoneSpace
    scaleMode=BMM_Additive
    scaleSpace=BCS_BoneSpace
    translationMode=BMM_Additive
    translationSpace=BCS_BoneSpace
    alphaBoolBlend {blendOption=Linear}
    boneToModify {boneName=dyn_snakeHead_01_mid}
    rotation {pitch=-15}
   }
   n13 = "Transform (Modify) Bone - Bone: dyn_snakeHead_00_lft"(ComponentPose=@n12, Alpha=1, Translation="0,0,0", Rotation=Get Snake Movement(), Scale="1,1,1") {
    rotationMode=BMM_Additive
    rotationSpace=BCS_BoneSpace
    scaleMode=BMM_Additive
    scaleSpace=BCS_BoneSpace
    translationMode=BMM_Additive
    translationSpace=BCS_BoneSpace
    alphaBoolBlend {blendOption=Linear}
    boneToModify {boneName=dyn_snakeHead_00_lft}
   }
   n14 = "Transform (Modify) Bone - Bone: dyn_snakeTail_00_lft_MSTR"(ComponentPose=@n13, Alpha=1, Translation="0,0,0", Rotation="0,0,50", Scale="1,1,1") {
    rotationMode=BMM_Additive
    rotationSpace=BCS_BoneSpace
    scaleMode=BMM_Additive
    scaleSpace=BCS_BoneSpace
    translationMode=BMM_Additive
    translationSpace=BCS_BoneSpace
    alphaBoolBlend {blendOption=Linear}
    boneToModify {boneName=dyn_snakeTail_00_lft_MSTR}
    rotation {roll=50}
   }
   n15 = RigidBody(ComponentPose=@n14, Alpha=0.2) {
    bOverrideWorldGravity=true
    overridePhysicsAsset=&/F_MED_Thrive_Spirit_Pack_Physics
    simulationSpace=BaseBoneSpace
    alphaBoolBlend {blendOption=Linear}
    baseBoneRef {boneName=pack}
    componentAppliedLinearAccClamp=<5000 5000 5000>
    componentLinearAccScale=<0.2 0.2 0.2>
    componentLinearVelScale=<0.2 0.2 0.2>
    overrideWorldGravity {z=-980}
   }
   n16 = "Transform (Modify) Bone - Bone: root"(ComponentPose=@n15, Translation=Get backpack_translation(), Rotation=Get backpack_rotate()) {alpha=0, rotationMode=BMM_Additive, translationMode=BMM_Additive, alphaBoolBlend {blendOption=Linear}, boneToModify {boneName=root}}
   Output Pose(Result=Component To Local(ComponentPose=@n16)) {layerGroup=None}
   Trail controller - Bone: dyn_snakeTongue_02_lft(Alpha=0.4) {
    chainLength=3
    stretchLimit=0.3
    rotationLimits [{limitMax=<180 180 180>}, {limitMax=<180 180 180>}, {limitMax=<180 180 180>}]
    rotationOffsets [<0 0 0>, <0 0 0>, <0 0 0>]
    trailBone {boneName=dyn_snakeTongue_02_lft}
    trailRelaxationSpeed {editorCurveData {keys [<at 0 10 RCIM_Linear>,<at 0.2425 9.5109 RCIM_Linear>,<at 1 5 RCIM_Linear>]}}
   }
  ]}
 EventGraph {nodes [
   Event BlueprintUpdateAnimation()
   > n1 = Random Vector Function!(
    execute
    CharacterPartAnimInstance:Custom Character Part Anim Instance Object Reference=Self-Reference()
    SkelMeshComponent:Mesh Component Object Reference
    JointName:Name=None
    GameTimeInSeconds:double=GetGameTimeInSeconds()
    TimeOffset:double=0
    PerlinScale:double=0
    Frequency:Vector="0.2,1.1,0.75"
    Amplitude:Vector="1,20,20"
    Offset:Vector="0,0,0"
   ) -> (then, VectorOut:Vector)
   n2 = BreakVector(InVec=@n1)
   > Set!(execute, Snake Movement_Roll:float=@n2, Snake Movement_Pitch:float=@n2.Y, Snake Movement_Yaw:float=@n2.Z) -> (then, Output_Get:Rotator)
   > n3 = Branch(Condition=Get bIsInFrontEnd())
   n4 = GetOwnerSkeletalMeshComponent()
   n5 = BreakRotator(InRot=BreakTransform(InTransform=GetDeltaTransformFromRefPose(self=@n4, BoneName=spine_05, BaseName=spine_01)).Rotation)
   Set backpack_rotate(
    execute=@n3.else
    backpack_rotate=MakeRotator(
     Roll=MapRangeClamped(Value=@n5.Yaw, InRangeA=0, InRangeB=-35, OutRangeB=40)
     Pitch=MapRangeClamped(Value=@n5.Yaw, InRangeA=0, InRangeB=-25, OutRangeB=10)
    )
   )
   > Set backpack_translation(backpack_translation=MakeVector(Y=MapRangeClamped(Value=@n5.Yaw, InRangeA=0, InRangeB=-3, OutRangeB=-2)))
   n6 = Event OnCharacterCustomizationLoaded()
   n7 = Get Dynamic Material Instance(execute=[@n6.then, @n15.then], SkeletalMeshComponent=GetPartSkeletalMeshComponent())
   > n8 = Sequence()
   n9 = PickaxeItemDefinitionHasAnyMetaTag(Tags=Make Array("[0]"=(TagName="Cosmetics.MetaTags.ThriveFemale")), bRequiresEquipped=true)
   > n10 = Branch(Condition=@n9)
   > n11 = Branch(Condition=Get bIsSkyDiving())
   SetScalarParameterValue(execute=[@n13.then, @n11.then, @n12.else, @n10.else], self=@n7, ParameterName=Hide Element 01, Value=0)
   SetScalarParameterValue(execute=[@n11.else, @n12.then], self=@n7, ParameterName=Hide Element 01, Value=1)
   n12 = Branch(execute=[@n14.then, @n8.then_1], Condition=@n9)
   n13 = Event OnBeginSkydiving()
   n14 = Event OnEndSkydiving()
   n15 = Event OnWeaponChanged()
   TryGetPawnOwner()
  ]}
]
variables [
 backpack_rotate {type=Rotator}
 backpack_translation {type=Vector}
 Snake Movement {type=Rotator}
]
